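# Masked material for Football event field logos, painting a logo texture onto grass with worn muddy areas.
# The logo, shiftable along V, is multiplied by brightened red-channel grass detail from MF_Landscape_GrassDistanceBlend, which also drives specular.
# A world-aligned, Z-projected mud layer blends over it through HeightLerp on vertex red, with vertex green mixing two mud tints.
# Logo alpha sets the final opacity mask, so mud stays within the logo shape.
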Material /Figment_S08/Athena/Events/Football/Materials/M_Football_Field_Logos
graph {
 properties {bUseMaterialAttributes=true, blendMode=BLEND_Masked, pixelDepthOffsetMode=PDOM_Legacy, usedShadingModels=Default Lit}
 nodes [
  n1 = Constant()
  n2 = ScalarParameter() {group=Logos, parameterName=offset}
  n3 = TextureSampleParameter2D(UVs=Add(A=TextureCoordinate(), B=AppendVector(A=@n1, B=@n2))) {group=Logos, parameterName=Texture, texture=/Figment_S08/Athena/Events/Football/Textures/T_Football_Logos_D}
  n4 = MaterialFunctionCall!(UseVertexInterp) {materialFunction=/Game/Athena/Environments/Landscape/MaterialFunctions/MF_Landscape_GrassDistanceBlend}
  n5 = ComponentMask(None=@n4.Grass RGB) {r=true}
  n6 = ScalarParameter() {defaultValue=5, parameterName=bright}
  n7 = Multiply(A=@n3.RGB, B=Saturate(None=Multiply(A=@n5, B=@n6)))
  n8 = ComponentMask(None=AppendVector(A=@n4.Grass RGB, B=@n4.Grass RGB)) {a=true}
  n9 = LinearInterpolate(
   B=ScalarParameter() {defaultValue=0.15, group=Grass, parameterName=GrassSpecAmount}
   Alpha=Power(Base=@n8, Exp=ScalarParameter() {defaultValue=2, group=Grass, parameterName=GrassSpecPower})
  )
  n10 = Constant() {r=0.9}
  n11 = OneMinus(None=@n8)
  n12 = VectorParameter() {group=Mud, parameterName=MudTint2, defaultValue=<rgba 0.1719 0.1192 0.0387 0>}
  n13 = VectorParameter() {group=Mud, parameterName=MudTint, defaultValue=<rgba 0.3073 0.2619 0.2113 0>}
  n14 = VertexColor()
  n15 = TextureObjectParameter() {group=Mud, parameterName=Mud_Diffuse, samplerSource=SSM_Wrap_WorldGroupSettings, texture=/Game/Packages/DS_Fortnight/WhiteRockCustom_D}
  n16 = TextureObjectParameter() {
   group=Mud
   parameterName=Mud_Normal
   samplerSource=SSM_Wrap_WorldGroupSettings
   samplerType=SAMPLERTYPE_Normal
   texture=/Game/Packages/DS_Fortnight/WhiteRockCustom_N
  }
  n17 = MaterialFunctionCall!(
   Diffuse Texture Object=@n15
   Spec Texture Object
   TextureSize=Abs(None=ScalarParameter() {defaultValue=900, group=Mud, parameterName=MudTextureTilingFrequency})
   --------------------------
   Use seperate settings for Z projection
   Use seperate texture for Z projection
   Texture Size Z projection
   Z Plane Diffuse Texture Object
   Z Plane Spec Texture Object
   "--------------------------      "
   ProjectionTransitionContrast
   Use Texture Alpha Heightmap to Refine the Transistion
   HeightLerpTransitionContrast
   PickTheDiffuseHeightChannel
   TheDiffuseHeightChannel
   "--------------------------       "
   NormalTextureObject=@n16
   World Space Normals
   "--------------------------  "
   Export Float 4 Diffuse=StaticBool() {value=true}
   Export Float 4 Spec
   "--------------------------     "
   WorldPosition
   World Space Normal
  ) {materialFunction=/Engine/Functions/Engine_MaterialFunctions02/Texturing/WorldAlignedTextures_Complex}
  n18 = ComponentMask(None=@n17.Z Diffuse Texture) {a=true}
  n19 = ScalarParameter() {defaultValue=0.5574, parameterName=contrast}
  n20 = SetMaterialAttributes!(
   MaterialAttributes
   Base Color=Multiply(
    A=MaterialFunctionCall(A=@n12.RGB, B=@n13.RGB, Transition Phase=@n14.G, Height Texture=@n18, Contrast=@n19) {materialFunction=&HeightLerp}.Results
    B=ComponentMask(None=@n17.Z Diffuse Texture) {b=true, g=true, r=true}
   )
   Roughness=ScalarParameter() {defaultValue=0.8, group=Mud, parameterName=MudRoughScalar}
   Specular=ScalarParameter() {defaultValue=0.5, group=Mud, parameterName=MudSpecScalar}
   Normal=@n17.Normal Z Texture
   Opacity=@n18
  ) {attributeSetTypes [
    #1
    #2
    #3
    #4
    #5
   ]}
  n21 = VertexColor()
  SetMaterialAttributes!(
   MaterialAttributes=BlendMaterialAttributes(
    A=MakeMaterialAttributes!(
     BaseColor=@n7
     Metallic
     Specular=@n9
     Roughness=@n10
     Anisotropy
     EmissiveColor
     Opacity=@n11
     OpacityMask
     Normal
     Tangent
     WorldPositionOffset
     SubsurfaceColor
     ClearCoat
     ClearCoatRoughness
     AmbientOcclusion
     Refraction
     CustomizedUVs_0
     CustomizedUVs_1
     CustomizedUVs_2
     CustomizedUVs_3
     CustomizedUVs_4
     CustomizedUVs_5
     CustomizedUVs_6
     CustomizedUVs_7
     PixelDepthOffset
     ShadingModel
     Displacement
    )
    B=@n20
    Alpha=MaterialFunctionCall(Transition Phase=@n21.R, Height Texture=@n18, Contrast=@n19) {materialFunction=&HeightLerp}.Alpha
   )
   Opacity Mask=@n3.A
  ) {attributeSetTypes [#6]}
 ]
}
